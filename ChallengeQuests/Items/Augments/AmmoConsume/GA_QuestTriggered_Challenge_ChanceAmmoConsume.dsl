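; Challenge quest gameplay ability that, when triggered by a gameplay event, has a small random chance to consume a ranged weapon's remaining ammo.
; It calls the parent activation, casts the event's target to FortWeaponRanged, and on a roll of roughly one in ten consumes the amount returned by GetRemainingAmmo.
; The ability ends after the roll, but a failed cast does nothing and does not end it in this graph.
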(event Ability|EventActivateAbilityFromEvent
  (params (EventData :type Gameplay%20Event%20Data%20Structure%20%28by%20ref%29))
  (bind (_eventtag _instigator _target) (Utilities|Struct|BreakGameplayEventData EventData))
  (|CallParentFunction EventData)
  (bind _asfort_weapon_ranged (Utilities|Casting|CastToFortWeaponRanged _target)
    (:then
      (if (< (Math|Random|RandomFloatinRange 1.0 1e+02) 1e+01)
        (FortWeapon|Ammo|ConsumeAmmo _asfort_weapon_ranged (FortWeapon|GetRemainingAmmo _asfort_weapon_ranged) false)
        (else
          _))
      (Ability|EndAbility))
    (:CastFailed
      _)))
